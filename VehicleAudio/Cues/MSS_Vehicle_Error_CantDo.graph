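# One-shot mono MetaSound cue that plays a "can't do" UI error ping for vehicles.
# OnPlay passes through a short Trigger Delay into a Wave Player, whose finish trigger ends the source; the Attack Time input is unconnected.
# It uses the non-critical in-game UI sound class and the vehicle boost-end concurrency set.
# Its volume is modulated by the vehicle boost sidechain control bus.
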MetaSoundSource /VehicleAudio/Cues/MSS_Vehicle_Error_CantDo
inputs [
 Attack Time {type=Time}
 UE.Source.OnPlay {type=Trigger}
]
interfaces [UE.OutputFormat.Mono:1.0, UE.Source.OneShot:1.0, UE.Source:1.1]
outputs [
 "UE.OutputFormat.Mono.Audio:0" {type=Audio}
 UE.Source.OneShot.OnFinished {type=Trigger}
]
pages [Default {
  input_defaults {Attack Time=0.01, UE.Source.OnPlay=null}
  nodes [
   TriggerDelay_0 = UE.Trigger Delay(In=^UE.Source.OnPlay) {also {defaults {Delay Time=0.05}}}
   WavePlayer_0 = UE.Wave Player:Mono(Play=@TriggerDelay_0.Out) {also {defaults {Wave Asset=/Game/Sounds/Fort_UI_Sounds/Gameplay/UI_Error_Ping_Cant_Do_Default_01}, version [1,1]}}
  ]
  links [{dst="@UE.Source.OneShot.OnFinished", src=WavePlayer_0.On Finished}, {dst="@UE.OutputFormat.Mono.Audio:0", src=WavePlayer_0.Out Mono}]
 }]
properties {
 assetImportData=&self:AssetImportData
 sampleRate=48000
 soundClassObject=&Fort_UI_InGame_NonCritical
 concurrencyOverrides {maxCount=1, resolutionRule=PreventNew, retriggerTime=1.5}
 concurrencySet [/VehicleAudio/Settings/Conc_Valet_Vehicle_BoostEnd]
 modulationSettings {
  volumeRouting=Override
  volumeModulationDestination {bEnableModulation=true, modulators [&/CB_Valet_Vehicle_Boost_SideChain]}
 }
}
