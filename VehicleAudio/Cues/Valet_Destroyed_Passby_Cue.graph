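# Sound cue for the destroyed Valet vehicle passby, built from in-air explosion flame sounds.
# A Random node picks one of three equally weighted flame variations and feeds a Modulator at the root, left at default settings.
# It uses the shared vehicle movement attenuation and the exterior vehicle sound class.
# A concurrency override allows one instance per owner, with a retrigger time to limit rapid repeats.
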SoundCue /VehicleAudio/Cues/Valet_Destroyed_Passby_Cue
root=Modulator_1
nodes [
 {id=Modulator_1, children [Random_3]}
 {id=Random_3, children [WavePlayer_8, WavePlayer_9, WavePlayer_10], properties {weights [1,1,1]}}
 {id=WavePlayer_10, properties {soundWaveAssetPtr=/VehicleAudio/Assets/Damage/Veh_Valet_Damage_Explode_InAir_Flame_03}}
 {id=WavePlayer_8, properties {soundWaveAssetPtr=/VehicleAudio/Assets/Damage/Veh_Valet_Damage_Explode_InAir_Flame_01}}
 {id=WavePlayer_9, properties {soundWaveAssetPtr=/VehicleAudio/Assets/Damage/Veh_Valet_Damage_Explode_InAir_Flame_02}}
]
properties {
 attenuationSettings=/VehicleAudio/Settings/Vehicle_Movement_Att
 bOverrideConcurrency=true
 maxDistance=12512
 soundClassObject=&Fort_Vehicles_Exterior
 volumeMultiplier=1
 concurrencyOverrides {bLimitToOwner=true, maxCount=1, retriggerTime=1}
}
