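# Sound cue for the Valet vehicle's exterior boost, layering a one-shot start with looping jet whine and jet wash.
# A FuelRemaining parameter drives continuous modulators that drop both loops' volume and pitch as fuel runs low.
# The whine chains a second, inverse pitch modulator, so its pitch climbs as fuel burns before the low-fuel drop.
# Owner-limited concurrency, boost attenuation and a vehicle boost bus send control playback.
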SoundCue /VehicleAudio/Cues/Veh_Valet_Boost_Cue
root=Mixer_1
nodes [
 {id=Mixer_1, children [ModulatorContinuous_2, ModulatorContinuous_0, WavePlayer_3], properties {inputVolume [1.5,1,1]}}
 {
  id=ModulatorContinuous_0
  children [WavePlayer_4]
  properties {
   pitchModulationParams {maxInput=0.4, minOutput=0.75, parameterName=FuelRemaining}
   volumeModulationParams {maxInput=0.4, minOutput=0.75, parameterName=FuelRemaining}
  }
 }
 {
  id=ModulatorContinuous_1
  children [WavePlayer_5]
  properties {pitchModulationParams {maxOutput=0.8, minOutput=1.2, parameterName=FuelRemaining}}
 }
 {
  id=ModulatorContinuous_2
  children [ModulatorContinuous_1]
  properties {
   pitchModulationParams {maxInput=0.2, minOutput=0.7, parameterName=FuelRemaining}
   volumeModulationParams {maxInput=0.2, minOutput=0.5, parameterName=FuelRemaining}
  }
 }
 {id=WavePlayer_3, properties {soundWaveAssetPtr=/VehicleAudio/Assets/Boost/Veh_Valet_Boost_Start_01}}
 {id=WavePlayer_4, properties {bLooping=true, soundWaveAssetPtr=/VehicleAudio/Assets/Boost/Veh_Valet_Boost_Loop_Jet_Wash_01}}
 {id=WavePlayer_5, properties {bLooping=true, soundWaveAssetPtr=/VehicleAudio/Assets/Boost/Veh_Valet_Boost_Loop_Jet_Whine_01}}
]
properties {
 attenuationSettings=/VehicleAudio/Settings/Vehicle_Movement_Boost_Att
 maxDistance=12000
 soundClassObject=&Fort_Vehicles_Exterior
 volumeMultiplier=0.7
 concurrencyOverrides {bLimitToOwner=true, maxCount=2, volumeScale=0.5}
 concurrencySet [/VehicleAudio/Settings/Vehicle_Engine_Conc, /VehicleAudio/Settings/Conc_Valet_Vehicle_BoostEnd]
 preEffectBusSends [{audioBus=&/AB_Vehicle_Boost, soundSourceBus=null}]
}
