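# Sound cue for the end of the Valet vehicle's boost, playing a single wave through the exterior vehicle sound class.
# It uses the shared vehicle boost attenuation and sends pre-effect audio to the vehicle boost audio bus.
# An inline concurrency override keeps one instance per owner by stopping the oldest, taking the place of the assigned engine concurrency set.
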SoundCue /VehicleAudio/Cues/Veh_Valet_Boost_End_Cue
root=WavePlayer_3
nodes [{id=WavePlayer_3, properties {soundWaveAssetPtr=/VehicleAudio/Assets/Boost/Veh_Valet_Boost_End_01}}]
properties {
 attenuationSettings=/VehicleAudio/Settings/Vehicle_Movement_Boost_Att
 bOverrideConcurrency=true
 maxDistance=12000
 soundClassObject=&Fort_Vehicles_Exterior
 volumeMultiplier=0.7
 concurrencyOverrides {bLimitToOwner=true, maxCount=1, resolutionRule=StopOldest, volumeScale=0.5}
 concurrencySet [/VehicleAudio/Settings/Vehicle_Engine_Conc]
 preEffectBusSends [{audioBus=&/AB_Vehicle_Boost, sendLevel=2, soundSourceBus=null}]
}
