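# Sound cue for heavy body collision impacts on the Valet vehicle, played in the exterior vehicle sound class with body-collision concurrency.
# Picks randomly among five equally weighted impact waves, then modulates pitch and volume slightly for variation.
# A bIsLocal branch sends local playback through a non-spatialized, unattenuated 2D path, while false or unset uses a slightly quieter close-impact attenuation path.
# Virtualization is disabled.
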SoundCue /VehicleAudio/Cues/Veh_Valet_Collision_Body_Hv_Cue
root=Modulator_1
nodes [
 {id=Attenuation_0, children [Random_3], properties {attenuationSettings=/VehicleAudio/Settings/Vehicle_Collision_Impact_Close_Att}}
 {
  id=Attenuation_1
  children [Random_3]
  properties {attenuationSettings=/VehicleAudio/Settings/Vehicle_Collision_Impact_Close_2D_Att, attenuationOverrides {bAttenuate=false, bSpatialize=false}}
 }
 {id=Branch_1, children [Mixer_1,Mixer_0,Mixer_0], properties {boolParameterName=bIsLocal}}
 {id=Mixer_0, children [Attenuation_0], properties {inputVolume [0.8]}}
 {id=Mixer_1, children [Attenuation_1], properties {inputVolume [1]}}
 {id=Modulator_1, children [Branch_1], properties {pitchMax=1.1, pitchMin=1, volumeMax=1, volumeMin=0.8}}
 {id=Random_3, children [WavePlayer_13, WavePlayer_14, WavePlayer_15, WavePlayer_16, WavePlayer_17], properties {weights [1,1,1,1,1]}}
 {id=WavePlayer_13, properties {soundWaveAssetPtr=/VehicleAudio/Assets/Collision/Veh_Valet_Impt_Hv_01}}
 {id=WavePlayer_14, properties {soundWaveAssetPtr=/VehicleAudio/Assets/Collision/Veh_Valet_Impt_Hv_02}}
 {id=WavePlayer_15, properties {soundWaveAssetPtr=/VehicleAudio/Assets/Collision/Veh_Valet_Impt_Hv_03}}
 {id=WavePlayer_16, properties {soundWaveAssetPtr=/VehicleAudio/Assets/Collision/Veh_Valet_Impt_Hv_04}}
 {id=WavePlayer_17, properties {soundWaveAssetPtr=/VehicleAudio/Assets/Collision/Veh_Valet_Impt_Hv_05}}
]
properties {
 maxDistance=12000
 soundClassObject=&Fort_Vehicles_Exterior
 virtualizationMode=Disabled
 volumeMultiplier=0.8
 concurrencySet [/VehicleAudio/Settings/Collision_Body_Conc]
}
