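# Sound cue for medium body collision impacts on the Valet vehicle, in the exterior vehicle sound class.
# A random pick of five impact waves feeds two attenuation paths.
# A bIsLocal branch selects a non-spatialized, non-attenuated 2D path through a mixer when true, and close-impact attenuation when false or unset.
# A root modulator varies pitch and volume; body-collision concurrency governs instances, with high priority and virtualization disabled.
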SoundCue /VehicleAudio/Cues/Veh_Valet_Collision_Body_Med_Cue
root=Modulator_1
nodes [
 {id=Attenuation_0, children [Random_2], properties {attenuationSettings=/VehicleAudio/Settings/Vehicle_Collision_Impact_Close_Att}}
 {
  id=Attenuation_1
  children [Random_2]
  properties {attenuationSettings=/VehicleAudio/Settings/Vehicle_Collision_Impact_Close_2D_Att, attenuationOverrides {bAttenuate=false, bSpatialize=false}}
 }
 {id=Branch_0, children [Mixer_0, Attenuation_0, Attenuation_0], properties {boolParameterName=bIsLocal}}
 {id=Mixer_0, children [Attenuation_1], properties {inputVolume [1]}}
 {id=Modulator_1, children [Branch_0], properties {pitchMax=1.2, volumeMax=1, volumeMin=0.5}}
 {id=Random_2, children [WavePlayer_8, WavePlayer_9, WavePlayer_10, WavePlayer_11, WavePlayer_12], properties {weights [1,1,1,1,1]}}
 {id=WavePlayer_10, properties {soundWaveAssetPtr=/VehicleAudio/Assets/Collision/Veh_Valet_Impt_Med_03}}
 {id=WavePlayer_11, properties {soundWaveAssetPtr=/VehicleAudio/Assets/Collision/Veh_Valet_Impt_Med_04}}
 {id=WavePlayer_12, properties {soundWaveAssetPtr=/VehicleAudio/Assets/Collision/Veh_Valet_Impt_Med_05}}
 {id=WavePlayer_8, properties {soundWaveAssetPtr=/VehicleAudio/Assets/Collision/Veh_Valet_Impt_Med_01}}
 {id=WavePlayer_9, properties {soundWaveAssetPtr=/VehicleAudio/Assets/Collision/Veh_Valet_Impt_Med_02}}
]
properties {
 maxDistance=12000
 priority=0.9
 soundClassObject=&Fort_Vehicles_Exterior
 virtualizationMode=Disabled
 volumeMultiplier=0.8
 concurrencySet [/VehicleAudio/Settings/Collision_Body_Conc]
}
