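# Sound cue for the Valet vehicle's body collision impact involving a player.
# A mixer layers a random player-impact sound with a quieter, slightly delayed random heavy impact, each with its own attenuation.
# A bIsLocal branch routes the local case through lower volume and 2D attenuation, while others use the mix directly.
# Slight random pitch varies the output, and concurrency limits with a retrigger time restrict overlapping hits.
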SoundCue /VehicleAudio/Cues/Veh_Valet_Collision_Body_Player_Cue
root=Modulator_1
nodes [
 {
  id=Attenuation_0
  children [Mixer_0]
  properties {attenuationSettings=/VehicleAudio/Settings/Vehicle_Collision_Impact_Close_2D_Att, attenuationOverrides {bAttenuate=false, bSpatialize=false}}
 }
 {id=Attenuation_1, children [Delay_0], properties {attenuationSettings=/VehicleAudio/Settings/Vehicle_Collision_Impact_Close_Att}}
 {id=Attenuation_4, children [Random_0], properties {attenuationSettings=/VehicleAudio/Settings/Vehicle_Collision_Impact_Player_Att}}
 {id=Branch_0, children [Modulator_0,Mixer_0,Mixer_0], properties {boolParameterName=bIsLocal}}
 {id=Delay_0, children [Random_3], properties {delayMax=0.05, delayMin=0.05}}
 {id=Mixer_0, children [Attenuation_4, Attenuation_1], properties {inputVolume [1.35,0.85]}}
 {id=Modulator_0, children [Attenuation_0], properties {pitchMax=1, pitchMin=1, volumeMax=0.8, volumeMin=0.8}}
 {id=Modulator_1, children [Branch_0], properties {pitchMax=1.1, pitchMin=1, volumeMax=1, volumeMin=1}}
 {id=Random_0, children [WavePlayer_0,WavePlayer_1,WavePlayer_3], properties {bShouldExcludeFromBranchCulling=true, weights [1,1,1]}}
 {id=Random_3, children [WavePlayer_13, WavePlayer_14, WavePlayer_15, WavePlayer_16, WavePlayer_17], properties {weights [1,1,1,1,1]}}
 {id=WavePlayer_0, properties {soundWaveAssetPtr=/VehicleAudio/Assets/Collision/Veh_Valet_Impt_Player_01}}
 {id=WavePlayer_1, properties {soundWaveAssetPtr=/VehicleAudio/Assets/Collision/Veh_Valet_Impt_Player_02}}
 {id=WavePlayer_13, properties {soundWaveAssetPtr=/VehicleAudio/Assets/Collision/Veh_Valet_Impt_Hv_01}}
 {id=WavePlayer_14, properties {soundWaveAssetPtr=/VehicleAudio/Assets/Collision/Veh_Valet_Impt_Hv_02}}
 {id=WavePlayer_15, properties {soundWaveAssetPtr=/VehicleAudio/Assets/Collision/Veh_Valet_Impt_Hv_03}}
 {id=WavePlayer_16, properties {soundWaveAssetPtr=/VehicleAudio/Assets/Collision/Veh_Valet_Impt_Hv_04}}
 {id=WavePlayer_17, properties {soundWaveAssetPtr=/VehicleAudio/Assets/Collision/Veh_Valet_Impt_Hv_05}}
 {id=WavePlayer_3, properties {soundWaveAssetPtr=/VehicleAudio/Assets/Collision/Veh_Valet_Impt_Player_03}}
]
properties {
 attenuationSettings=&Vehicle_Collision_Impact_Close_Att
 bOverrideConcurrency=true
 maxDistance=12000
 soundClassObject=&Fort_2DHitNotifications_Instigate_PSM
 virtualizationMode=Disabled
 volumeMultiplier=0.8
 concurrencyOverrides {maxCount=1, retriggerTime=0.5, voiceStealReleaseTime=0.5}
 concurrencySet [/VehicleAudio/Settings/Collision_Body_Conc]
}
