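# Sound cue for the Valet vehicle's heavy suspension bounce on collision, randomly picking one of seven waves with modulated pitch and volume.
# A branch on bIsLocal sends the local case through a boosted, non-spatialized, non-attenuated 2D path, while false or unset uses close suspension attenuation.
# Both paths share one random node, which is excluded from branch culling.
# It plays in the exterior vehicle sound class under a shared suspension collision concurrency set.
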SoundCue /VehicleAudio/Cues/Veh_Valet_Collision_Susp_Hv_Cue
root=Modulator_2
nodes [
 {id=Attenuation_0, children [Random_0], properties {attenuationSettings=/VehicleAudio/Settings/Vehicle_Movement_Close_Suspension_Att}}
 {
  id=Attenuation_1
  children [Random_0]
  properties {attenuationSettings=/VehicleAudio/Settings/Vehicle_Collision_Impact_Close_2D_Att, attenuationOverrides {bAttenuate=false, bSpatialize=false}}
 }
 {id=Branch_1, children [Mixer_1, Attenuation_0, Attenuation_0], properties {boolParameterName=bIsLocal}}
 {id=Mixer_1, children [Attenuation_1], properties {inputVolume [1.25]}}
 {id=Modulator_2, children [Branch_1], properties {pitchMax=1.2, pitchMin=1, volumeMax=1, volumeMin=0.65}}
 {
  id=Random_0
  children [WavePlayer_3,WavePlayer_4,WavePlayer_5,WavePlayer_6,WavePlayer_7,WavePlayer_8,WavePlayer_9]
  properties {bShouldExcludeFromBranchCulling=true, weights [1,1,1,1,1,1,1]}
 }
 {id=WavePlayer_3, properties {soundWaveAssetPtr=/VehicleAudio/Assets/Suspension/Veh_Valet_Susp_Bounce_Hv_02}}
 {id=WavePlayer_4, properties {soundWaveAssetPtr=/VehicleAudio/Assets/Suspension/Veh_Valet_Susp_Bounce_Hv_03}}
 {id=WavePlayer_5, properties {soundWaveAssetPtr=/VehicleAudio/Assets/Suspension/Veh_Valet_Susp_Bounce_Hv_04}}
 {id=WavePlayer_6, properties {soundWaveAssetPtr=/VehicleAudio/Assets/Suspension/Veh_Valet_Susp_Bounce_Hv_07}}
 {id=WavePlayer_7, properties {soundWaveAssetPtr=/VehicleAudio/Assets/Suspension/Veh_Valet_Susp_Bounce_Hv_01}}
 {id=WavePlayer_8, properties {soundWaveAssetPtr=/VehicleAudio/Assets/Suspension/Veh_Valet_Susp_Bounce_Hv_05}}
 {id=WavePlayer_9, properties {soundWaveAssetPtr=/VehicleAudio/Assets/Suspension/Veh_Valet_Susp_Bounce_Hv_06}}
]
properties {
 maxDistance=8000
 soundClassObject=&Fort_Vehicles_Exterior
 virtualizationMode=Disabled
 volumeMultiplier=1
 concurrencyOverrides {bLimitToOwner=true, maxCount=1, resolutionRule=PreventNew, retriggerTime=1}
 concurrencySet [/VehicleAudio/Settings/Collision_Susp_Conc]
}
