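# Sound Cue for the Valet vehicle's light suspension bounce collision sound, randomly picking one of six equally weighted samples.
# A bIsLocal branch plays it without attenuation or spatialization when true, and through close suspension attenuation when false or unset.
# A root modulator randomizes pitch slightly upward and volume across the full zero-to-one range.
# It uses the exterior vehicle sound class and suspension collision concurrency, with virtualization disabled.
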SoundCue /VehicleAudio/Cues/Veh_Valet_Collision_Susp_Lt_Cue
root=Modulator_2
nodes [
 {id=Attenuation_0, children [Random_0], properties {attenuationSettings=/VehicleAudio/Settings/Vehicle_Movement_Close_Suspension_Att}}
 {
  id=Attenuation_1
  children [Random_0]
  properties {attenuationSettings=/VehicleAudio/Settings/Vehicle_Collision_Impact_Close_2D_Att, attenuationOverrides {bAttenuate=false, bSpatialize=false}}
 }
 {id=Branch_0, children [Attenuation_1, Attenuation_0, Attenuation_0], properties {boolParameterName=bIsLocal}}
 {id=Modulator_2, children [Branch_0], properties {pitchMax=1.1, pitchMin=1, volumeMax=1, volumeMin=0}}
 {
  id=Random_0
  children [WavePlayer_3,WavePlayer_4,WavePlayer_5,WavePlayer_6,WavePlayer_7,WavePlayer_8]
  properties {bShouldExcludeFromBranchCulling=true, weights [1,1,1,1,1,1]}
 }
 {id=WavePlayer_3, properties {soundWaveAssetPtr=/VehicleAudio/Assets/Suspension/Veh_Valet_Susp_Bounce_Lt_01}}
 {id=WavePlayer_4, properties {soundWaveAssetPtr=/VehicleAudio/Assets/Suspension/Veh_Valet_Susp_Bounce_Lt_02}}
 {id=WavePlayer_5, properties {soundWaveAssetPtr=/VehicleAudio/Assets/Suspension/Veh_Valet_Susp_Bounce_Lt_03}}
 {id=WavePlayer_6, properties {soundWaveAssetPtr=/VehicleAudio/Assets/Suspension/Veh_Valet_Susp_Bounce_Lt_04}}
 {id=WavePlayer_7, properties {soundWaveAssetPtr=/VehicleAudio/Assets/Suspension/Veh_Valet_Susp_Bounce_Lt_05}}
 {id=WavePlayer_8, properties {soundWaveAssetPtr=/VehicleAudio/Assets/Suspension/Veh_Valet_Susp_Bounce_Lt_06}}
]
properties {
 maxDistance=8000
 priority=0.8
 soundClassObject=&Fort_Vehicles_Exterior
 virtualizationMode=Disabled
 volumeMultiplier=0.5
 concurrencySet [/VehicleAudio/Settings/Collision_Susp_Conc]
}
