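# Sound cue for medium suspension bounce collisions on the Valet vehicle, playing one of six equally weighted variations.
# A Branch on bIsLocal sends true through a slightly lowered Mixer into non-spatialized, non-attenuated 2D playback, and false or unset into close suspension attenuation.
# Both paths share the Random node, which is excluded from branch culling.
# It uses the exterior vehicle sound class and suspension collision concurrency, with virtualization disabled.
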SoundCue /VehicleAudio/Cues/Veh_Valet_Collision_Susp_Med_Cue
root=Branch_0
nodes [
 {id=Attenuation_0, children [Random_1], properties {attenuationSettings=/VehicleAudio/Settings/Vehicle_Movement_Close_Suspension_Att}}
 {
  id=Attenuation_1
  children [Random_1]
  properties {attenuationSettings=/VehicleAudio/Settings/Vehicle_Collision_Impact_Close_2D_Att, attenuationOverrides {bAttenuate=false, bSpatialize=false}}
 }
 {id=Branch_0, children [Mixer_1, Attenuation_0, Attenuation_0], properties {boolParameterName=bIsLocal}}
 {id=Mixer_1, children [Attenuation_1], properties {inputVolume [0.9]}}
 {
  id=Random_1
  children [WavePlayer_9, WavePlayer_10, WavePlayer_11, WavePlayer_12, WavePlayer_13, WavePlayer_14]
  properties {bShouldExcludeFromBranchCulling=true, weights [1,1,1,1,1,1]}
 }
 {id=WavePlayer_10, properties {soundWaveAssetPtr=/VehicleAudio/Assets/Suspension/Veh_Valet_Susp_Bounce_Med_02}}
 {id=WavePlayer_11, properties {soundWaveAssetPtr=/VehicleAudio/Assets/Suspension/Veh_Valet_Susp_Bounce_Med_03}}
 {id=WavePlayer_12, properties {soundWaveAssetPtr=/VehicleAudio/Assets/Suspension/Veh_Valet_Susp_Bounce_Med_04}}
 {id=WavePlayer_13, properties {soundWaveAssetPtr=/VehicleAudio/Assets/Suspension/Veh_Valet_Susp_Bounce_Med_05}}
 {id=WavePlayer_14, properties {soundWaveAssetPtr=/VehicleAudio/Assets/Suspension/Veh_Valet_Susp_Bounce_Med_06}}
 {id=WavePlayer_9, properties {soundWaveAssetPtr=/VehicleAudio/Assets/Suspension/Veh_Valet_Susp_Bounce_Med_01}}
]
properties {
 maxDistance=8000
 priority=0.9
 soundClassObject=&Fort_Vehicles_Exterior
 virtualizationMode=Disabled
 volumeMultiplier=0.7
 concurrencySet [/VehicleAudio/Settings/Collision_Susp_Conc]
}
