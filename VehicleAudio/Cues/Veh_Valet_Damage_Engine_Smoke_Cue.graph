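# Sound cue for the Valet vehicle's damaged-engine smoke, layering a looping smoke sound with a one-shot start sound.
# The loop fades in over about a second, while a Branch on IsDamaged skips the start sound when true.
# Each layer has its own attenuation and sound class; the mix is modulated, then enveloped down to half volume over time.
# Concurrency allows one voice per owner, stopping the oldest, alongside shared damage-state concurrency sets.
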SoundCue /VehicleAudio/Cues/Veh_Valet_Damage_Engine_Smoke_Cue
root=Enveloper_0
nodes [
 {id=Attenuation_0, children [SoundClass_0], properties {attenuationSettings=/VehicleAudio/Settings/Vehicle_Damage_Att}}
 {id=Attenuation_1, children [SoundClass_1], properties {attenuationSettings=/VehicleAudio/Settings/Vehicle_Damage_Smoke_Loop_Att}}
 {id=Branch_0, children [null, Attenuation_0, Attenuation_0], properties {boolParameterName=IsDamaged}}
 {
  id=Enveloper_0
  children [Modulator_0]
  properties {pitchCurve {editorCurveData {keys [{time=0}]}}, volumeCurve {editorCurveData {keys [<at 0 1>,<at 10 1>,<at 25 0.5>]}}}
 }
 {
  id=Enveloper_1
  children [WavePlayer_0]
  properties {pitchCurve {editorCurveData {keys [{time=0}]}}, volumeCurve {editorCurveData {keys [<at 0 0>,<at 0.8586 0.2529>,<at 1.25 1>]}}}
 }
 {id=Mixer_2, children [Attenuation_1, Branch_0], properties {inputVolume [1,0.6]}}
 {id=Modulator_0, children [Mixer_2]}
 {id=SoundClass_0, children [WavePlayer_1], properties {soundClassOverride=/Game/Sounds/FortSoundClassesAndModes/Fort_Explosives_Large_PSM}}
 {id=SoundClass_1, children [Enveloper_1], properties {soundClassOverride=&Fort_Vehicles_Exterior}}
 {id=WavePlayer_0, properties {bLooping=true, soundWaveAssetPtr=/VehicleAudio/Assets/Damage/Veh_Valet_Damage_Engine_Smoke_Loop_01}}
 {id=WavePlayer_1, properties {soundWaveAssetPtr=/VehicleAudio/Assets/Damage/Veh_Valet_Damage_Engine_Smoke_Start_01}}
]
properties {
 maxDistance=6000
 volumeMultiplier=0.85
 concurrencyOverrides {bLimitToOwner=true, maxCount=1, resolutionRule=StopOldest, voiceStealReleaseTime=0.2}
 concurrencySet [/VehicleAudio/Settings/Conc_DamageState_Global, /VehicleAudio/Settings/Conc_DamageState_Owner]
}
