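# Sound cue for the Valet vehicle's destruction explosion, layered by distance.
# A top mixer blends close, medium and distant branches, each with its own attenuation, sound class override and randomly chosen waves.
# The close branch mixes a pitch-modulated explosion with a randomly picked roar that follows after a short random delay.
# The cue overrides concurrency to cap simultaneous instances.
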SoundCue /VehicleAudio/Cues/Veh_Valet_Destroy_Cue
root=Mixer_4
nodes [
 {
  id=Attenuation_0
  children [Random_2]
  properties {
   attenuationSettings=/Game/Sounds/Attenuation/ATT_Vehicle_Explosion_Distant
   attenuationOverrides {falloffDistance=3000, attenuationShapeExtents {x=1000}}
  }
 }
 {
  id=Attenuation_4
  children [Mixer_1]
  properties {attenuationSettings=/Game/Sounds/Attenuation/ATT_Vehicle_Explosion_Close, attenuationOverrides {falloffDistance=1600}}
 }
 {
  id=Attenuation_5
  children [Random_1]
  properties {
   attenuationSettings=/Game/Sounds/Attenuation/ATT_Vehicle_Explosion_Medium
   attenuationOverrides {falloffDistance=3000, attenuationShapeExtents {x=1000}}
  }
 }
 {id=Delay_0, children [Random_3], properties {delayMax=0.3, delayMin=0.175}}
 {id=Mixer_0, children [SoundClass_3], properties {inputVolume [1]}}
 {id=Mixer_1, children [Modulator_2,Delay_0], properties {inputVolume [1,0.8]}}
 {id=Mixer_4, children [Mixer_0,SoundClass_2,SoundClass_0], properties {inputVolume [0.5,1,1.5]}}
 {id=Modulator_2, children [Random_0], properties {pitchMax=1.1, pitchMin=0.9, volumeMax=1, volumeMin=1}}
 {id=Random_0, children [WavePlayer_0,WavePlayer_1], properties {bShouldExcludeFromBranchCulling=true, weights [1,1]}}
 {id=Random_1, children [WavePlayer_2,WavePlayer_3], properties {bShouldExcludeFromBranchCulling=true, weights [1,1]}}
 {id=Random_2, children [WavePlayer_4,WavePlayer_5], properties {bShouldExcludeFromBranchCulling=true, weights [1,1]}}
 {
  id=Random_3
  children [WavePlayer_7, WavePlayer_8, WavePlayer_9, WavePlayer_10, WavePlayer_11]
  properties {bShouldExcludeFromBranchCulling=true, weights [1,1,1,1,1]}
 }
 {
  id=SoundClass_0
  children [Attenuation_0]
  properties {soundClassOverride=&Fort_WeaponRanged_Fire_Distant}
 }
 {
  id=SoundClass_2
  children [Attenuation_5]
  properties {soundClassOverride=&Fort_WeaponRanged_Fire_Distant}
 }
 {id=SoundClass_3, children [Attenuation_4], properties {soundClassOverride=/Game/Sounds/FortSoundClassesAndModes/Fort_Explosives_Large_PSM}}
 {id=WavePlayer_0, properties {soundWaveAssetPtr=/VehicleAudio/Assets/Damage/Veh_Valet_Damage_Explode_Close_01}}
 {id=WavePlayer_1, properties {soundWaveAssetPtr=/VehicleAudio/Assets/Damage/Veh_Valet_Damage_Explode_Close_02}}
 {id=WavePlayer_10, properties {soundWaveAssetPtr=/VehicleAudio/Assets/Damage/Veh_Valet_Damage_Explode_Close_Roar_02}}
 {id=WavePlayer_11, properties {soundWaveAssetPtr=/VehicleAudio/Assets/Damage/Veh_Valet_Damage_Explode_Close_Roar_01}}
 {id=WavePlayer_2, properties {soundWaveAssetPtr=/VehicleAudio/Assets/Damage/Veh_Valet_Damage_Explode_Med_01}}
 {id=WavePlayer_3, properties {soundWaveAssetPtr=/VehicleAudio/Assets/Damage/Veh_Valet_Damage_Explode_Med_02}}
 {id=WavePlayer_4, properties {soundWaveAssetPtr=/VehicleAudio/Assets/Damage/Veh_Valet_Damage_Explode_Distant_01}}
 {id=WavePlayer_5, properties {soundWaveAssetPtr=/VehicleAudio/Assets/Damage/Veh_Valet_Damage_Explode_Distant_02}}
 {id=WavePlayer_7, properties {soundWaveAssetPtr=/VehicleAudio/Assets/Damage/Veh_Valet_Damage_Explode_Close_Roar_05}}
 {id=WavePlayer_8, properties {soundWaveAssetPtr=/VehicleAudio/Assets/Damage/Veh_Valet_Damage_Explode_Close_Roar_04}}
 {id=WavePlayer_9, properties {soundWaveAssetPtr=/VehicleAudio/Assets/Damage/Veh_Valet_Damage_Explode_Close_Roar_03}}
]
properties {bOverrideConcurrency=true, maxDistance=27500, volumeMultiplier=1, concurrencyOverrides {maxCount=8}}
