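# Sound cue for the foley played when exiting the Valet sedan vehicle.
# A single wave player feeds the exit foley sample through a Modulator node left at default settings.
# It routes to the vehicle interior sound class, borrows the golf cart close-movement attenuation, and shares an enter/exit concurrency group.
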SoundCue /VehicleAudio/Cues/Veh_Valet_Exit_Sedan_Cue
root=Modulator_0
nodes [
 {id=Modulator_0, children [WavePlayer_0]}
 {id=WavePlayer_0, properties {soundWaveAssetPtr=/VehicleAudio/Assets/Foley/Veh_Valet_Sedan_V4_01_Foley_Exit_01}}
]
properties {
 attenuationSettings=/Game/Sounds/Attenuation/Fort_GolfCart_Movement_Close_Att
 maxDistance=5300
 soundClassObject=&Fort_Vehicles_Interior
 volumeMultiplier=1
 concurrencySet [/VehicleAudio/Settings/EnterExit_Conc]
}
