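# Sound cue for the Valet vehicle's fuel-gain audio, played as a spatialized vehicle exterior sound.
# A mixer layers a one-shot fill-start wave with a looping fill wave at slightly lower volume, both beginning together.
# The loop has no end inside the cue, so it plays until stopped externally.
# An overridden concurrency rule caps instances per owner, stopping the oldest and scaling down volume.
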SoundCue /VehicleAudio/Cues/Veh_Valet_GainFuel_Cue
root=Mixer_0
nodes [
 {id=Mixer_0, children [WavePlayer_2,WavePlayer_0], properties {inputVolume [1,0.75]}}
 {id=WavePlayer_0, properties {bLooping=true, soundWaveAssetPtr=/VehicleAudio/Assets/Fuel/Veh_Valet_Fuel_Fill_Loop_01}}
 {id=WavePlayer_2, properties {soundWaveAssetPtr=/VehicleAudio/Assets/Fuel/Veh_Valet_Fuel_Fill_Start_01}}
]
properties {
 attenuationSettings=/Game/Sounds/Attenuation/Fort_Prop_Vehicles_Att
 bOverrideConcurrency=true
 maxDistance=3000
 soundClassObject=&Fort_Vehicles_Exterior
 volumeMultiplier=0.5
 attenuationOverrides {falloffDistance=10}
 concurrencyOverrides {bLimitToOwner=true, maxCount=2, resolutionRule=StopOldest, volumeScale=0.5, volumeScaleAttackTime=0.1}
}
