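# Sound cue for the Valet vehicle's windshield glass breaking on impact.
# A random node picks one of three equally weighted break samples and feeds a modulator that varies pitch and volume.
# It plays in the exterior vehicle sound class with close-detail focus attenuation and no virtualization.
# A concurrency override limits overlapping breaks, stopping the farthest first and throttling rapid retriggers.
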SoundCue /VehicleAudio/Cues/Veh_Valet_Glass_Break_Cue
root=Modulator_0
nodes [
 {id=Modulator_0, children [Random_0], properties {pitchMax=1.1, volumeMax=1}}
 {id=Random_0, children [WavePlayer_0,WavePlayer_1,WavePlayer_2], properties {bShouldExcludeFromBranchCulling=true, weights [1,1,1]}}
 {id=WavePlayer_0, properties {soundWaveAssetPtr=/VehicleAudio/Assets/Collision/Veh_Valet_Impt_Glass_Windshield_Break_01}}
 {id=WavePlayer_1, properties {soundWaveAssetPtr=/VehicleAudio/Assets/Collision/Veh_Valet_Impt_Windshield_Break_02}}
 {id=WavePlayer_2, properties {soundWaveAssetPtr=/VehicleAudio/Assets/Collision/Veh_Valet_Impt_Windshield_Break_03}}
]
properties {
 attenuationSettings=/VehicleAudio/Settings/Vehicle_Movement_Close_Detail_Focus_Att
 bOverrideConcurrency=true
 maxDistance=4300
 soundClassObject=&Fort_Vehicles_Exterior
 virtualizationMode=Disabled
 volumeMultiplier=1
 concurrencyOverrides {maxCount=2, resolutionRule=StopFarthestThenPreventNew, retriggerTime=0.1}
}
