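# Sound cue for the Valet vehicle's out-of-fuel tank warning in the vehicle audio set.
# A single wave player feeds the root with one empty-fuel warning wave.
# It uses the shared vehicle prop attenuation and the interior vehicle sound class.
# A concurrency override limits it to one instance per owner, so a new trigger does not play while one is still sounding.
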SoundCue /VehicleAudio/Cues/Veh_Valet_OutOfFuel_Tank_Cue
root=WavePlayer_0
nodes [{id=WavePlayer_0, properties {soundWaveAssetPtr=/VehicleAudio/Assets/Fuel/Veh_Valet_Fuel_Empty_Warning_01}}]
properties {
 attenuationSettings=/Game/Sounds/Attenuation/Fort_Prop_Vehicles_Att
 bOverrideConcurrency=true
 maxDistance=3000
 soundClassObject=&Fort_Vehicles_Interior
 volumeMultiplier=0.8
 concurrencyOverrides {bLimitToOwner=true, maxCount=1, resolutionRule=PreventNew}
}
