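# Sound cue for the Valet vehicle's rev limiter, playing a backfire sound at reduced volume.
# A short random delay feeds a modulator at default settings, which varies a single wave player's output.
# It uses close-detail vehicle movement attenuation and the vehicle interior sound class.
# A concurrency override caps it at two simultaneous instances, stopping the oldest first.
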SoundCue /VehicleAudio/Cues/Veh_Valet_RevLimiter_Cue
root=Delay_0
nodes [
 {id=Delay_0, children [Modulator_0], properties {delayMax=0.05}}
 {id=Modulator_0, children [WavePlayer_0]}
 {id=WavePlayer_0, properties {soundWaveAssetPtr=/VehicleAudio/Assets/Gears/Veh_Valet_Gen_RevLimit_Backfire}}
]
properties {
 attenuationSettings=/VehicleAudio/Settings/Vehicle_Movement_Close_Detail_Att
 bOverrideConcurrency=true
 maxDistance=12000
 soundClassObject=&Fort_Vehicles_Interior
 volumeMultiplier=0.2
 concurrencyOverrides {maxCount=2, resolutionRule=StopOldest}
}
