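# Sound cue for the Valet vehicle's mechanical gear shift sound.
# A Random node picks one of three equally weighted gear shift waves, then feeds a root Modulator at default settings for playback variation.
# The cue plays at reduced volume through the vehicle interior sound class, with a close-detail vehicle movement attenuation asset.
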SoundCue /VehicleAudio/Cues/Veh_Valet_Shift_Cue
root=Modulator_0
nodes [
 {id=Modulator_0, children [Random_1]}
 {id=Random_1, children [WavePlayer_2,WavePlayer_3,WavePlayer_4], properties {weights [1,1,1]}}
 {id=WavePlayer_2, properties {soundWaveAssetPtr=/VehicleAudio/Assets/Gears/Veh_Valet_Gen_Gear_Shift_Mech_03}}
 {id=WavePlayer_3, properties {soundWaveAssetPtr=/VehicleAudio/Assets/Gears/Veh_Valet_Gen_Gear_Shift_Mech_01}}
 {id=WavePlayer_4, properties {soundWaveAssetPtr=/VehicleAudio/Assets/Gears/Veh_Valet_Gen_Gear_Shift_Mech_02}}
]
properties {
 attenuationSettings=/VehicleAudio/Settings/Vehicle_Movement_Close_Detail_Att
 maxDistance=12000
 soundClassObject=&Fort_Vehicles_Interior
 volumeMultiplier=0.25
}
