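# Sound cue for the Valet sedan's gear shift sound in the vehicle audio set.
# A Random node picks among four equally weighted shift recordings, and a Modulator at the root passes the result out.
# It plays quieter than full volume through the interior vehicles sound class, using a close-detail vehicle movement attenuation.
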SoundCue /VehicleAudio/Cues/Veh_Valet_Shift_Sedan_Cue
root=Modulator_0
nodes [
 {id=Modulator_0, children [Random_0]}
 {id=Random_0, children [WavePlayer_0,WavePlayer_1,WavePlayer_5,WavePlayer_6], properties {weights [1,1,1,1]}}
 {id=WavePlayer_0, properties {soundWaveAssetPtr=/VehicleAudio/Assets/Gears/Veh_Valet_Sedan_Shift_01}}
 {id=WavePlayer_1, properties {soundWaveAssetPtr=/VehicleAudio/Assets/Gears/Veh_Valet_Sedan_Shift_02}}
 {id=WavePlayer_5, properties {soundWaveAssetPtr=/VehicleAudio/Assets/Gears/Veh_Valet_Sedan_Shift_03}}
 {id=WavePlayer_6, properties {soundWaveAssetPtr=/VehicleAudio/Assets/Gears/Veh_Valet_Sedan_Shift_04}}
]
properties {
 attenuationSettings=/VehicleAudio/Settings/Vehicle_Movement_Close_Detail_Att
 maxDistance=12000
 soundClassObject=&Fort_Vehicles_Interior
 volumeMultiplier=0.7
}
