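# Sound cue for the Valet sports car gear shift, routed to the interior vehicle sound class with close-detail movement attenuation.
# A mixer layers three randomized, modulated sets: an immediate mechanical shift, exhaust pops after a short delay, and a quieter V10 shift slightly later.
# The exhaust pop layer is pitched down, and the staggered delays spread each shift into a brief sequence.
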SoundCue /VehicleAudio/Cues/Veh_Valet_Shift_Sport_Cue
root=Mixer_0
nodes [
 {id=Delay_0, children [Modulator_0], properties {delayMax=0.15, delayMin=0.15}}
 {id=Delay_1, children [Modulator_2], properties {delayMax=0.1, delayMin=0.1}}
 {id=Mixer_0, children [Modulator_1,Delay_0,Delay_1], properties {inputVolume [1,0.4,1]}}
 {id=Modulator_0, children [Random_0]}
 {id=Modulator_1, children [Random_2]}
 {id=Modulator_2, children [Random_3], properties {pitchMax=0.9, pitchMin=0.8}}
 {id=Random_0, children [WavePlayer_0,WavePlayer_1,WavePlayer_5,WavePlayer_6], properties {weights [1,1,1,1]}}
 {id=Random_2, children [WavePlayer_2,WavePlayer_3,WavePlayer_4], properties {weights [1,1,1]}}
 {id=Random_3, children [WavePlayer_7, WavePlayer_8, WavePlayer_9, WavePlayer_10], properties {weights [1,1,1,1]}}
 {id=WavePlayer_0, properties {soundWaveAssetPtr=/VehicleAudio/Assets/Gears/Veh_Valet_Sports_Car_V10_01_Shift_01}}
 {id=WavePlayer_1, properties {soundWaveAssetPtr=/VehicleAudio/Assets/Gears/Veh_Valet_Sports_Car_V10_01_Shift_02}}
 {id=WavePlayer_10, properties {soundWaveAssetPtr=/VehicleAudio/Assets/Gears/Veh_Valet_Sports_Car_ExaustPop_04}}
 {id=WavePlayer_2, properties {soundWaveAssetPtr=/VehicleAudio/Assets/Gears/Veh_Valet_Gen_Gear_Shift_Mech_03}}
 {id=WavePlayer_3, properties {soundWaveAssetPtr=/VehicleAudio/Assets/Gears/Veh_Valet_Gen_Gear_Shift_Mech_01}}
 {id=WavePlayer_4, properties {soundWaveAssetPtr=/VehicleAudio/Assets/Gears/Veh_Valet_Gen_Gear_Shift_Mech_02}}
 {id=WavePlayer_5, properties {soundWaveAssetPtr=/VehicleAudio/Assets/Gears/Veh_Valet_Sports_Car_V10_01_Shift_03}}
 {id=WavePlayer_6, properties {soundWaveAssetPtr=/VehicleAudio/Assets/Gears/Veh_Valet_Sports_Car_V10_01_Shift_04}}
 {id=WavePlayer_7, properties {soundWaveAssetPtr=/VehicleAudio/Assets/Gears/Veh_Valet_Sports_Car_ExaustPop_01}}
 {id=WavePlayer_8, properties {soundWaveAssetPtr=/VehicleAudio/Assets/Gears/Veh_Valet_Sports_Car_ExaustPop_02}}
 {id=WavePlayer_9, properties {soundWaveAssetPtr=/VehicleAudio/Assets/Gears/Veh_Valet_Sports_Car_ExaustPop_03}}
]
properties {
 attenuationSettings=/VehicleAudio/Settings/Vehicle_Movement_Close_Detail_Att
 maxDistance=12000
 soundClassObject=&Fort_Vehicles_Interior
 volumeMultiplier=1
}
