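# One-shot sound cue for short Valet vehicle tire skids, played as attenuated exterior vehicle audio.
# A Surface integer switch picks a random pool of asphalt or dirt skid waves, with the dirt pool shared by two slots and other slots silent.
# Continuous modulators fade volume out as TiresPopped rises and trim it slightly as LocalPlayerTime grows.
# A root modulator adds slight random pitch variation.
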SoundCue /VehicleAudio/Cues/Veh_Valet_Skid_OneShot_Cue
root=Modulator_0
nodes [
 {
  id=ModulatorContinuous_0
  children [Switch_0]
  properties {volumeModulationParams {default=0, maxInput=4, maxOutput=0, minOutput=1, parameterName=TiresPopped}}
 }
 {
  id=ModulatorContinuous_1
  children [ModulatorContinuous_0]
  properties {volumeModulationParams {default=0, maxInput=40, maxOutput=0.8, minInput=20, minOutput=1, parameterName=LocalPlayerTime}}
 }
 {id=Modulator_0, children [ModulatorContinuous_1], properties {pitchMax=1.1, pitchMin=0.9}}
 {id=Random_0, children [WavePlayer_8, WavePlayer_9, WavePlayer_10, WavePlayer_11], properties {weights [1,1,1,1]}}
 {
  id=Random_3
  children [WavePlayer_4, WavePlayer_5, WavePlayer_7, WavePlayer_0, WavePlayer_1, WavePlayer_2, WavePlayer_12, WavePlayer_13, WavePlayer_14]
  properties {weights [1,1,1,1,1,1,1,1,1]}
 }
 {id=Switch_0, children [null,Random_3,null,null,null,null,null,null,null,null,Random_0,Random_0], properties {intParameterName=Surface}}
 {id=WavePlayer_0, properties {soundWaveAssetPtr=/VehicleAudio/Assets/Skid/Veh_Valet_Skid_Gen_Asphalt_Short_06}}
 {id=WavePlayer_1, properties {soundWaveAssetPtr=/VehicleAudio/Assets/Skid/Veh_Valet_Skid_Gen_Asphalt_Short_07}}
 {id=WavePlayer_10, properties {soundWaveAssetPtr=/VehicleAudio/Assets/Skid/Veh_Valet_Skid_Gen_Dirt_Short_03}}
 {id=WavePlayer_11, properties {soundWaveAssetPtr=/VehicleAudio/Assets/Skid/Veh_Valet_Skid_Gen_Dirt_Short_04}}
 {id=WavePlayer_12, properties {soundWaveAssetPtr=/VehicleAudio/Assets/Skid/Veh_Valet_Skid_Gen_Asphalt_Short_09}}
 {id=WavePlayer_13, properties {soundWaveAssetPtr=/VehicleAudio/Assets/Skid/Veh_Valet_Skid_Gen_Asphalt_Short_10}}
 {id=WavePlayer_14, properties {soundWaveAssetPtr=/VehicleAudio/Assets/Skid/Veh_Valet_Skid_Gen_Asphalt_Short_11}}
 {id=WavePlayer_2, properties {soundWaveAssetPtr=/VehicleAudio/Assets/Skid/Veh_Valet_Skid_Gen_Asphalt_Short_08}}
 {id=WavePlayer_4, properties {soundWaveAssetPtr=/VehicleAudio/Assets/Skid/Veh_Valet_Skid_Gen_Asphalt_Short_02}}
 {id=WavePlayer_5, properties {soundWaveAssetPtr=/VehicleAudio/Assets/Skid/Veh_Valet_Skid_Gen_Asphalt_Short_03}}
 {id=WavePlayer_7, properties {soundWaveAssetPtr=/VehicleAudio/Assets/Skid/Veh_Valet_Skid_Gen_Asphalt_Short_05}}
 {id=WavePlayer_8, properties {soundWaveAssetPtr=/VehicleAudio/Assets/Skid/Veh_Valet_Skid_Gen_Dirt_Short_01}}
 {id=WavePlayer_9, properties {soundWaveAssetPtr=/VehicleAudio/Assets/Skid/Veh_Valet_Skid_Gen_Dirt_Short_02}}
]
properties {
 attenuationSettings=/VehicleAudio/Settings/Vehicle_Movement_Att
 maxDistance=12512
 soundClassObject=&Fort_Vehicles_Exterior
 volumeMultiplier=0.5
}
