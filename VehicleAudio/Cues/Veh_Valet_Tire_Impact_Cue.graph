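# Sound cue for a tire impact on the Valet vehicle, layering rubber and small junk impact sounds.
# A mixer combines one of two randomly chosen, sharply pitched-up rubber hits with a short junk hit.
# The mix passes through slight random modulation and a brief random start delay.
# A concurrency override allows one instance at a time, with small-object demolition attenuation, the impacts sound class and virtualization disabled.
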SoundCue /VehicleAudio/Cues/Veh_Valet_Tire_Impact_Cue
root=Delay_0
nodes [
 {id=Delay_0, children [Modulator_2], properties {delayMax=0.04, delayMin=0.02}}
 {id=Delay_1, children [WavePlayer_6]}
 {id=Mixer_0, children [Modulator_1,Delay_1], properties {inputVolume [1,1]}}
 {id=Modulator_1, children [Random_31], properties {pitchMax=1.6, pitchMin=1.5}}
 {id=Modulator_2, children [Mixer_0], properties {pitchMax=1.02}}
 {id=Random_31, children [WavePlayer_92, WavePlayer_7], properties {weights [1,1]}}
 {id=WavePlayer_6, properties {soundWaveAssetPtr=/Game/Sounds/Fort_Impact_Sounds/Demolition/Junk/Junk_ImpactShort_Small03}}
 {id=WavePlayer_7, properties {soundWaveAssetPtr=/Game/Sounds/Fort_Impact_Sounds/Demolition/Rubber/Rubber_Impact01}}
 {id=WavePlayer_92, properties {soundWaveAssetPtr=/Game/Sounds/Fort_Impact_Sounds/Demolition/Rubber/Rubber_Impact02}}
]
properties {
 attenuationSettings=/Game/Sounds/Attenuation/Fort_Demolition_Hit_Small_Sized_Object_Att
 bOverrideConcurrency=true
 maxDistance=4000
 soundClassObject=&Fort_Impacts
 virtualizationMode=Disabled
 volumeMultiplier=0.6
 concurrencyOverrides {maxCount=1}
}
