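# Sound cue for the tire blowout effect on Valet vehicles, played as an exterior vehicle sound with tire-pop attenuation.
# A short fixed delay feeds a modulator that slightly randomizes pitch, which wraps an equally weighted random pick among four blowout impact waves.
# The cue overrides concurrency to allow one instance at a time, disables virtualization, and boosts overall volume.
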SoundCue /VehicleAudio/Cues/Veh_Valet_Tire_Popped_Cue
root=Delay_0
nodes [
 {id=Delay_0, children [Modulator_0], properties {delayMax=0.04, delayMin=0.04}}
 {id=Modulator_0, children [Random_2], properties {pitchMax=1.1, pitchMin=0.9, volumeMax=1, volumeMin=1}}
 {
  id=Random_2
  children [WavePlayer_0, WavePlayer_1, WavePlayer_2, WavePlayer_12]
  properties {bShouldExcludeFromBranchCulling=true, weights [1,1,1,1]}
 }
 {id=WavePlayer_0, properties {soundWaveAssetPtr=/VehicleAudio/Assets/Collision/Veh_Valet_Impt_Tire_Blowout_01}}
 {id=WavePlayer_1, properties {soundWaveAssetPtr=/VehicleAudio/Assets/Collision/Veh_Valet_Impt_Tire_Blowout_02}}
 {id=WavePlayer_12, properties {soundWaveAssetPtr=/VehicleAudio/Assets/Collision/Veh_Valet_Impt_Tire_Blowout_04}}
 {id=WavePlayer_2, properties {soundWaveAssetPtr=/VehicleAudio/Assets/Collision/Veh_Valet_Impt_Tire_Blowout_03}}
]
properties {
 attenuationSettings=/Game/Athena/Sounds/Vehicles/Valet/Veh_Damage_Tire_Pop_Att
 bOverrideConcurrency=true
 maxDistance=11000
 soundClassObject=&Fort_Vehicles_Exterior
 virtualizationMode=Disabled
 volumeMultiplier=1.8
 concurrencyOverrides {maxCount=1}
}
