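# Sound cue for the Valet vehicle's low-fuel warning.
# A single wave player feeds the root directly with one warning sample at reduced volume in the vehicles sound class.
# It uses a super-close player attenuation asset and overrides concurrency so new instances are prevented rather than replacing a playing one.
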SoundCue /VehicleAudio/Cues/Veh_Valet_Warning_LowFuel_Cue
root=WavePlayer_2
nodes [{id=WavePlayer_2, properties {soundWaveAssetPtr=/VehicleAudio/Assets/Fuel/Veh_Valet_Fuel_Low_Warning_01}}]
properties {
 attenuationSettings=/Game/Sounds/Attenuation/Fort_Player_SuperClose_Att
 bOverrideConcurrency=true
 maxDistance=800
 soundClassObject=&Fort_Vehicles
 volumeMultiplier=0.6
 attenuationOverrides {distanceAlgorithm=NaturalSound, falloffDistance=4000, attenuationShapeExtents {x=1000}}
 concurrencyOverrides {resolutionRule=PreventNew}
}
