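# Sound cue for a vehicle landing impact, layering random landing sounds with pitch-varied heavy impacts under a concurrency override.
# A bIsLocal branch routes that shared layer either louder through non-spatialized, unattenuated player settings, or quieter through world land attenuation when false or unset.
# A second branch, mixed in parallel, adds random distant impacts with their own attenuation only for non-local or unset cases.
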SoundCue /VehicleAudio/Cues/Vehicle_Collision_Land_Cue
root=Mixer_1
nodes [
 {id=Attenuation_0, children [Mixer_0], properties {attenuationSettings=/VehicleAudio/Settings/Vehicle_Collision_Impact_Land_Att}}
 {id=Attenuation_2, children [Random_2], properties {attenuationSettings=/VehicleAudio/Settings/Vehicle_Collision_Impact_Distant_Att}}
 {
  id=Attenuation_3
  children [Mixer_0]
  properties {
   attenuationSettings=/VehicleAudio/Settings/Vehicle_Collision_Impact_Player_Att
   bOverrideAttenuation=true
   attenuationOverrides {bAttenuate=false, bSpatialize=false}
  }
 }
 {id=Branch_0, children [Mixer_2,Mixer_3,Mixer_3], properties {boolParameterName=bIsLocal}}
 {id=Branch_1, children [null, Attenuation_2, Attenuation_2], properties {boolParameterName=bIsLocal}}
 {id=Mixer_0, children [Random_0,Modulator_1], properties {inputVolume [1,0.7]}}
 {id=Mixer_1, children [Branch_0,Branch_1], properties {inputVolume [1,1]}}
 {id=Mixer_2, children [Attenuation_3], properties {inputVolume [0.7]}}
 {id=Mixer_3, children [Attenuation_0], properties {inputVolume [0.45]}}
 {id=Modulator_1, children [Random_1], properties {pitchMax=1.15, pitchMin=1, volumeMax=1, volumeMin=1}}
 {
  id=Random_0
  children [WavePlayer_0,WavePlayer_1,WavePlayer_2,WavePlayer_3]
  properties {bShouldExcludeFromBranchCulling=true, weights [1,1,1,1]}
 }
 {
  id=Random_1
  children [WavePlayer_5, WavePlayer_6, WavePlayer_7, WavePlayer_13, WavePlayer_14]
  properties {bShouldExcludeFromBranchCulling=true, weights [1,1,1,1,1]}
 }
 {id=Random_2, children [WavePlayer_8, WavePlayer_9, WavePlayer_10, WavePlayer_11, WavePlayer_12], properties {weights [1,1,1,1,1]}}
 {id=WavePlayer_0, properties {soundWaveAssetPtr=/VehicleAudio/Assets/Collision/Veh_Valet_Land_Default_01}}
 {id=WavePlayer_1, properties {soundWaveAssetPtr=/VehicleAudio/Assets/Collision/Veh_Valet_Land_Default_02}}
 {id=WavePlayer_10, properties {soundWaveAssetPtr=/VehicleAudio/Assets/Collision/Veh_Valet_Impt_Hv_Dist_03}}
 {id=WavePlayer_11, properties {soundWaveAssetPtr=/VehicleAudio/Assets/Collision/Veh_Valet_Impt_Hv_Dist_04}}
 {id=WavePlayer_12, properties {soundWaveAssetPtr=/VehicleAudio/Assets/Collision/Veh_Valet_Impt_Hv_Dist_05}}
 {id=WavePlayer_13, properties {soundWaveAssetPtr=/VehicleAudio/Assets/Collision/Veh_Valet_Impt_Hv_04}}
 {id=WavePlayer_14, properties {soundWaveAssetPtr=/VehicleAudio/Assets/Collision/Veh_Valet_Impt_Hv_05}}
 {id=WavePlayer_2, properties {soundWaveAssetPtr=/VehicleAudio/Assets/Collision/Veh_Valet_Land_Default_03}}
 {id=WavePlayer_3, properties {soundWaveAssetPtr=/VehicleAudio/Assets/Collision/Veh_Valet_Land_Default_04}}
 {id=WavePlayer_5, properties {soundWaveAssetPtr=/VehicleAudio/Assets/Collision/Veh_Valet_Impt_Hv_01}}
 {id=WavePlayer_6, properties {soundWaveAssetPtr=/VehicleAudio/Assets/Collision/Veh_Valet_Impt_Hv_02}}
 {id=WavePlayer_7, properties {soundWaveAssetPtr=/VehicleAudio/Assets/Collision/Veh_Valet_Impt_Hv_03}}
 {id=WavePlayer_8, properties {soundWaveAssetPtr=/VehicleAudio/Assets/Collision/Veh_Valet_Impt_Hv_Dist_01}}
 {id=WavePlayer_9, properties {soundWaveAssetPtr=/VehicleAudio/Assets/Collision/Veh_Valet_Impt_Hv_Dist_02}}
]
properties {
 bOverrideConcurrency=true
 maxDistance=12000
 soundClassObject=&Fort_Vehicles_Exterior
 volumeMultiplier=1
 concurrencyOverrides {maxCount=2, retriggerTime=0.15}
}
