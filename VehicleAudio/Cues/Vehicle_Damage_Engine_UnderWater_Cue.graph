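# Sound cue for an underwater damaged vehicle engine, layering a looping bubble bed with an explosion one-shot.
# The bubble loop fades in through an Enveloper, while a random pick of three pitched-up explosions plays unless the IsDamaged branch is true.
# Each layer has its own attenuation and sound class before mixing, then an outer envelope lowers volume over time.
# Concurrency limits it to one instance per owner, stopping the oldest.
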SoundCue /VehicleAudio/Cues/Vehicle_Damage_Engine_UnderWater_Cue
root=Enveloper_0
nodes [
 {id=Attenuation_0, children [Branch_0], properties {attenuationSettings=/VehicleAudio/Settings/Vehicle_Damage_Att}}
 {id=Attenuation_1, children [SoundClass_1], properties {attenuationSettings=/VehicleAudio/Settings/Vehicle_Damage_Smoke_Loop_Att}}
 {id=Branch_0, children [null,SoundClass_0,SoundClass_0], properties {boolParameterName=IsDamaged}}
 {
  id=Enveloper_0
  children [Modulator_0]
  properties {pitchCurve {editorCurveData {keys [{time=0}]}}, volumeCurve {editorCurveData {keys [<at 0 1>,<at 20 1>,<at 40 0.3>]}}}
 }
 {
  id=Enveloper_1
  children [WavePlayer_2]
  properties {pitchCurve {editorCurveData {keys [{time=0}]}}, volumeCurve {editorCurveData {keys [<at 0 0>,<at 0.8586 0.2529>,<at 1.25 1>]}}}
 }
 {id=Mixer_2, children [Attenuation_1, Attenuation_0], properties {inputVolume [0.75,0.5]}}
 {id=Modulator_0, children [Mixer_2]}
 {id=Modulator_1, children [Random_0], properties {pitchMax=1.5, pitchMin=1.4}}
 {id=Random_0, children [WavePlayer_3,WavePlayer_4,WavePlayer_5], properties {weights [1,1,1]}}
 {id=SoundClass_0, children [Modulator_1], properties {soundClassOverride=/Game/Sounds/FortSoundClassesAndModes/Fort_Explosives_Large_PSM}}
 {id=SoundClass_1, children [Enveloper_1], properties {soundClassOverride=&Fort_Vehicles_Exterior}}
 {id=WavePlayer_2, properties {bLooping=true, soundWaveAssetPtr=/Game/Athena/Sounds/Vehicles/Valet/Vehicle_Damaged_Underwater_Bubble_Loop_01}}
 {id=WavePlayer_3, properties {soundWaveAssetPtr=/Game/Athena/Sounds/Vehicles/Valet/Vehicle_Damage_Exple_Underwater_Close_01}}
 {id=WavePlayer_4, properties {soundWaveAssetPtr=/Game/Athena/Sounds/Vehicles/Valet/Vehicle_Damage_Exple_Underwater_Close_02}}
 {id=WavePlayer_5, properties {soundWaveAssetPtr=/Game/Athena/Sounds/Vehicles/Valet/Vehicle_Damage_Exple_Underwater_Close_03}}
]
properties {
 attenuationSettings=/Game/Athena/Sounds/Vehicles/Valet/Vehicle_Movement_Close_Detail_Att
 maxDistance=6000
 volumeMultiplier=0.5
 concurrencyOverrides {bLimitToOwner=true, maxCount=1, resolutionRule=StopOldest, voiceStealReleaseTime=0.2}
 concurrencySet [/VehicleAudio/Settings/Conc_DamageState_Global, /VehicleAudio/Settings/Conc_DamageState_Owner]
}
