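# Sound cue for the explosion heard when a vehicle is destroyed underwater.
# A Random node picks one of three equally weighted close underwater explosion waves and feeds a Modulator for variation.
# The root SoundClass node assigns the large-explosives sound class.
# The cue overrides attenuation with distance falloff, low-pass filtering over range, stereo spread and a non-spatialized near radius.
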SoundCue /VehicleAudio/Cues/Vehicle_Destroy_Underwater_Cue
root=SoundClass_3
nodes [
 {id=Modulator_0, children [Random_3]}
 {id=Random_3, children [WavePlayer_7,WavePlayer_8,WavePlayer_9], properties {weights [1,1,1]}}
 {id=SoundClass_3, children [Modulator_0], properties {soundClassOverride=/Game/Sounds/FortSoundClassesAndModes/Fort_Explosives_Large_PSM}}
 {id=WavePlayer_7, properties {soundWaveAssetPtr=/Game/Athena/Sounds/Vehicles/Valet/Vehicle_Damage_Exple_Underwater_Close_01}}
 {id=WavePlayer_8, properties {soundWaveAssetPtr=/Game/Athena/Sounds/Vehicles/Valet/Vehicle_Damage_Exple_Underwater_Close_02}}
 {id=WavePlayer_9, properties {soundWaveAssetPtr=/Game/Athena/Sounds/Vehicles/Valet/Vehicle_Damage_Exple_Underwater_Close_03}}
]
properties {
 bOverrideAttenuation=true
 maxDistance=6000
 volumeMultiplier=1
 attenuationOverrides {
  bAttenuateWithLPF=true
  falloffDistance=5000
  lPFFrequencyAtMax=2000
  lPFRadiusMax=5000
  lPFRadiusMin=1500
  nonSpatializedRadiusStart=200
  stereoSpread=500
  attenuationShapeExtents {x=1000}
 }
}
