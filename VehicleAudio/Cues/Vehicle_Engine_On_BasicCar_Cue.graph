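# Sound cue for the engine-on sound of the BasicCar vehicle.
# A single wave player root plays a sedan V8 start sample directly, with no randomization, modulation or mixing nodes.
# The cue routes to the Fort_Vehicles_Interior sound class and uses shared vehicle movement attenuation and vehicle engine concurrency settings.
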SoundCue /VehicleAudio/Cues/Vehicle_Engine_On_BasicCar_Cue
root=WavePlayer_0
nodes [{id=WavePlayer_0, properties {soundWaveAssetPtr=/VehicleAudio/Assets/Engines/Sedan/Veh_Valet_Sedan_V8_01_Start_01}}]
properties {
 attenuationSettings=/VehicleAudio/Settings/Vehicle_Movement_Att
 maxDistance=12512
 soundClassObject=&Fort_Vehicles_Interior
 volumeMultiplier=1
 concurrencySet [/VehicleAudio/Settings/Vehicle_Engine_Conc]
}
