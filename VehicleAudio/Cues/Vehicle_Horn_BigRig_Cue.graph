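# Sound cue for the big rig vehicle horn, playing a diesel truck horn wave.
# A single wave player feeds a continuous modulator that maps the PitchOffset parameter to a narrow pitch range, so the horn's pitch can be varied slightly at runtime.
# It uses shared horn attenuation, the exterior vehicle sound class and a boosted volume, and belongs to both engine and horn concurrency groups.
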SoundCue /VehicleAudio/Cues/Vehicle_Horn_BigRig_Cue
root=ModulatorContinuous_0
nodes [
 {
  id=ModulatorContinuous_0
  children [WavePlayer_0]
  properties {pitchModulationParams {default=0, maxOutput=1.05, minOutput=0.95, parameterName=PitchOffset}}
 }
 {id=WavePlayer_0, properties {soundWaveAssetPtr=/VehicleAudio/Assets/Horns/Veh_Valet_Truck_Diesel_01_Horn_01}}
]
properties {
 attenuationSettings=/VehicleAudio/Settings/Vehicle_Horn_Att
 maxDistance=16030.001
 soundClassObject=&Fort_Vehicles_Exterior
 volumeMultiplier=1.585
 concurrencySet [/VehicleAudio/Settings/Vehicle_Engine_Conc, /VehicleAudio/Settings/Vehicle_Horn_Conc]
}
