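# Sound cue for the sports car vehicle horn, played as exterior vehicle audio with horn attenuation.
# A Branch on the IsLooping boolean selects a looping 1000 Hz triangle test tone when true, apparently a placeholder.
# When the parameter is false or unset, it plays the one-shot valet sports car horn sample.
# Both engine and horn concurrency sets limit its playback.
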SoundCue /VehicleAudio/Cues/Vehicle_Horn_Sport_Cue
root=Branch_0
nodes [
 {id=Branch_0, children [WavePlayer_0,WavePlayer_1,WavePlayer_1], properties {boolParameterName=IsLooping}}
 {id=WavePlayer_0, properties {bLooping=true, soundWaveAssetPtr=/Game/Sounds/Fort_Test_Sounds/Tones/1000hz_Triangle_Tone_Loop}}
 {id=WavePlayer_1, properties {soundWaveAssetPtr=/VehicleAudio/Assets/Horns/Veh_Valet_SportsCar_Horn_02}}
]
properties {
 attenuationSettings=/VehicleAudio/Settings/Vehicle_Horn_Att
 maxDistance=16030.001
 soundClassObject=&Fort_Vehicles_Exterior
 volumeMultiplier=1.585
 concurrencySet [/VehicleAudio/Settings/Vehicle_Engine_Conc, /VehicleAudio/Settings/Vehicle_Horn_Conc]
}
