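# Sound cue for the SuperSport vehicle horn, playing a single sports car horn wave from the Valet vehicle horn assets.
# One wave player feeds the output directly, with no randomization or modulation.
# The cue uses shared horn attenuation, the exterior vehicle sound class, a raised volume, and both engine and horn concurrency sets to limit simultaneous playback.
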SoundCue /VehicleAudio/Cues/Vehicle_Horn_SuperSport_Cue
root=WavePlayer_0
nodes [{id=WavePlayer_0, properties {soundWaveAssetPtr=/VehicleAudio/Assets/Horns/Veh_Valet_SportsCar_Horn_01}}]
properties {
 attenuationSettings=/VehicleAudio/Settings/Vehicle_Horn_Att
 maxDistance=16030.001
 soundClassObject=&Fort_Vehicles_Exterior
 volumeMultiplier=1.585
 concurrencySet [/VehicleAudio/Settings/Vehicle_Engine_Conc, /VehicleAudio/Settings/Vehicle_Horn_Conc]
}
