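# Sound cue for the looping reverse warning beep of a vehicle, using a Battle Bus reverse beep wave.
# A single looping wave player feeds the output directly, with no modulation or randomization.
# It plays in the exterior vehicle sound class and uses a shared close-detail vehicle movement attenuation asset for distance falloff.
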SoundCue /VehicleAudio/Cues/Vehicle_Reverse_Beep_Loop_Cue
root=WavePlayer_1
nodes [{id=WavePlayer_1, properties {bLooping=true, soundWaveAssetPtr=/VehicleAudio/Assets/Horns/Veh_BattleBus_Reverse_Beeps_Loop_01}}]
properties {
 attenuationSettings=/VehicleAudio/Settings/Vehicle_Movement_Close_Detail_Att
 maxDistance=12000
 soundClassObject=&Fort_Vehicles_Exterior
 volumeMultiplier=1
 attenuationOverrides {falloffDistance=20000}
}
