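# Sound cue for a vehicle entering water, built from heavy Meatball vehicle water impact sounds.
# A Random node with equal weights picks one of six impact wave variations.
# The cue plays at raised pitch and reduced volume through the exterior vehicle sound class, using the Meatball distant collision attenuation.
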SoundCue /VehicleAudio/Cues/Vehicle_Water_Enter_Cue
root=Random_0
nodes [
 {
  id=Random_0
  children [WavePlayer_1, WavePlayer_2, WavePlayer_3, WavePlayer_4, WavePlayer_10, WavePlayer_11]
  properties {weights [1,1,1,1,1,1]}
 }
 {id=WavePlayer_1, properties {soundWaveAssetPtr=/Game/Athena/Sounds/Vehicles/Meatball/SoundWaves/Veh_Meatball_Water_Impact_Hv_06}}
 {id=WavePlayer_10, properties {soundWaveAssetPtr=/Game/Athena/Sounds/Vehicles/Meatball/SoundWaves/Veh_Meatball_Water_Impact_Hv_04}}
 {id=WavePlayer_11, properties {soundWaveAssetPtr=/Game/Athena/Sounds/Vehicles/Meatball/SoundWaves/Veh_Meatball_Water_Impact_Hv_05}}
 {id=WavePlayer_2, properties {soundWaveAssetPtr=/Game/Athena/Sounds/Vehicles/Meatball/SoundWaves/Veh_Meatball_Water_Impact_Hv_01}}
 {id=WavePlayer_3, properties {soundWaveAssetPtr=/Game/Athena/Sounds/Vehicles/Meatball/SoundWaves/Veh_Meatball_Water_Impact_Hv_02}}
 {id=WavePlayer_4, properties {soundWaveAssetPtr=/Game/Athena/Sounds/Vehicles/Meatball/SoundWaves/Veh_Meatball_Water_Impact_Hv_03}}
]
properties {
 attenuationSettings=/Game/Sounds/Attenuation/Fort_Meatball_Collision_Distant_Att
 maxDistance=10000
 pitchMultiplier=1.25
 soundClassObject=&Fort_Vehicles_Exterior
 volumeMultiplier=0.4
}
