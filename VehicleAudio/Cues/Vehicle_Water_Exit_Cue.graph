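# Sound cue for a vehicle water exit, named for that event and built from light water impact sounds of the Meatball vehicle.
# A Random node picks with equal weight among seven impact wave variations.
# It plays in the exterior vehicle sound class at reduced volume, using distant collision attenuation.
# Its own concurrency override limits simultaneous copies per owner and sets a short retrigger time to avoid stacking.
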SoundCue /VehicleAudio/Cues/Vehicle_Water_Exit_Cue
root=Random_0
nodes [
 {
  id=Random_0
  children [WavePlayer_0,WavePlayer_1,WavePlayer_2,WavePlayer_3,WavePlayer_4,WavePlayer_5,WavePlayer_6]
  properties {weights [1,1,1,1,1,1,1]}
 }
 {id=WavePlayer_0, properties {soundWaveAssetPtr=/Game/Athena/Sounds/Vehicles/Meatball/SoundWaves/Veh_Meatball_Water_Impact_Lt_01}}
 {id=WavePlayer_1, properties {soundWaveAssetPtr=/Game/Athena/Sounds/Vehicles/Meatball/SoundWaves/Veh_Meatball_Water_Impact_Lt_02}}
 {id=WavePlayer_2, properties {soundWaveAssetPtr=/Game/Athena/Sounds/Vehicles/Meatball/SoundWaves/Veh_Meatball_Water_Impact_Lt_03}}
 {id=WavePlayer_3, properties {soundWaveAssetPtr=/Game/Athena/Sounds/Vehicles/Meatball/SoundWaves/Veh_Meatball_Water_Impact_Lt_04}}
 {id=WavePlayer_4, properties {soundWaveAssetPtr=/Game/Athena/Sounds/Vehicles/Meatball/SoundWaves/Veh_Meatball_Water_Impact_Lt_05}}
 {id=WavePlayer_5, properties {soundWaveAssetPtr=/Game/Athena/Sounds/Vehicles/Meatball/SoundWaves/Veh_Meatball_Water_Impact_Lt_06}}
 {id=WavePlayer_6, properties {soundWaveAssetPtr=/Game/Athena/Sounds/Vehicles/Meatball/SoundWaves/Veh_Meatball_Water_Impact_Lt_07}}
]
properties {
 attenuationSettings=/Game/Sounds/Attenuation/Fort_Meatball_Collision_Distant_Att
 bOverrideConcurrency=true
 maxDistance=10000
 soundClassObject=&Fort_Vehicles_Exterior
 volumeMultiplier=0.65
 concurrencyOverrides {bLimitToOwner=true, maxCount=2, retriggerTime=0.5}
}
